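# Masked water material for streams and rivers, with optional waterfall, sand bank and lake-transition layers behind static switches.
# A flowmap or center-object direction drives world-projected foam normals that distort the pebble riverbed to fake refraction, with output passing through a render-to-texture water function.
# Vertex colors and height blends mask edges and layers, while world position offset adjusts water height or biases toward the camera against Z-fighting.
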Material /FigmentEnvironment/Materials/Water/Figment_Stream_Flowmaps_01
graph {
 properties {
  bDisableDepthTest=true
  bHasPixelAnimation=true
  bScreenSpaceReflections=true
  bTangentSpaceNormal=false
  bUseLightmapDirectionality=false
  bUseMaterialAttributes=true
  bUsedWithNanite=true
  bUsedWithSplineMeshes=true
  blendMode=BLEND_Masked
  physMaterial=/Game/FortressPhysicalMaterials/Water
  pixelDepthOffsetMode=PDOM_Legacy
  subsurfaceProfile=/Game/Effects/Fort_Effects/Materials/Water/WaterProfile
  translucencyLightingMode=TLM_SurfacePerPixelLighting
  translucencyPass=MTP_BeforeDOF
  translucentSelfShadowDensityScale=4
  translucentShadowDensityScale=1
  usedShadingModels=Default Lit
 }
 nodes [
  n1 = Constant() {r=1}
  n2 = VertexColor()
  n3 = TextureSampleParameter2D() {
   parameterName=FlowMap
   samplerType=SAMPLERTYPE_LinearColor
   texture=/Game/Packages/Fortnite_UniqueMaterials_1/Textures/T_River_Flowmap_Alpha_Straight_01
  }
  n4 = Multiply(A=@n2.A, B=@n3.A) {desc=Combine vert paint alpha}
  n5 = Multiply(A=@n4, B=ScalarParameter() {defaultValue=1.5, group=Edge Mask, parameterName=Alpha})
  n6 = ScalarParameter() {group=Edge Mask, parameterName=AlphaOffset}
  n7 = OneMinus(None=VertexColor() {desc="lake = black"}.B)
  n8 = Multiply(A=VertexColor() {desc="Red = Opacity"}.R, B=Add(A=Add(A=@n5, B=@n6), B=@n7) {desc=Blue Erases Alpha}) {desc=Paintable Alpha}
  n9 = Divide(A=ComponentMask(None=WorldPosition().XYZ) {g=true, r=true}, B=ScalarParameter() {defaultValue=512, parameterName=Sand Tiling})
  n10 = TextureSampleParameter2D!(UVs=@n9, Bias, Apply View MipBias) {
   desc=normal sample - under water and on the shoreline
   mipValueMode=TMVM_MipBias
   parameterName=RiverBedTexture_1
   texture=/Game/Packages/DS_Fortnight/pebbles
  }
  n11 = OneMinus(None=@n10.A)
  n12 = Multiply(
   A=Subtract(A=StaticSwitchParameter(True=@n10.A, False=@n11) {desc=HAACK, parameterName=Lake_Match}, B=Multiply(A=@n7) {constB=1.5})
   B=ScalarParameter() {group=Water Edge, parameterName=Water Edge Mult}
  )
  n13 = StaticSwitchParameter(
   True=MaterialFunctionCall(Transition Phase=@n8, Height Texture=@n12, Contrast=ScalarParameter() {defaultValue=0.5, group=Edge Mask, parameterName=AlphaContrast}) {materialFunction=&HeightLerp}.Alpha
   False=@n8
  ) {parameterName=UseEdgeHeightblend}
  n14 = MaterialFunctionCall(Alpha Threshold=@n13) {materialFunction=&DitherTemporalAA}
  n15 = QualitySwitch(Default=@n14.Result, Low=@n13, High=@n14.Result, Medium=@n13)
  n16 = StaticSwitchParameter(True=@n15, False=@n1) {group=Lake, parameterName=Enable Transition To Lake?}
  n17 = VertexColor()
  n18 = TextureCoordinate()
  n19 = ComponentMask(None=VectorParameter() {group=WaterFall, parameterName=EdgeScale_Large, defaultValue=<rgba 0.5 3 1 0>}.RGB) {g=true, r=true}
  n20 = Time() {bOverride_Period=true, period=9999999}
  n21 = TextureSample(UVs=Panner(Coordinate=Multiply(A=@n18, B=@n19), Time=@n20, Speed=VectorParameter() {parameterName=EdgeLarge_Speed, defaultValue=<rgba 2 0.05 0 0>}.RGB) {speedX=2, speedY=0.05}) {samplerType=SAMPLERTYPE_LinearColor, texture=/Game/Global/Textures/Default/Noise/T_Noise01}
  n22 = TextureCoordinate()
  n23 = ComponentMask(None=VectorParameter() {group=WaterFall, parameterName=EdgeScale_Detail, defaultValue=<rgba 1 22 1 0>}.RGB) {g=true, r=true}
  n24 = TextureSample(UVs=Panner(Coordinate=Multiply(A=@n22, B=@n23), Speed=VectorParameter() {parameterName=EdgeDetail_Speed, defaultValue=<rgba 1.333 -0.5 0 0>}.RGB) {speedX=1.333, speedY=-0.5}) {samplerType=SAMPLERTYPE_LinearColor, texture=&T_Noise01}
  n25 = Panner(Coordinate=TextureCoordinate() {uTiling=0.2, vTiling=0.2}, Time=@n20) {speedX=0.02, speedY=0.1}
  n26 = Clamp(None=LinearInterpolate(Alpha=TextureSample(UVs=@n25) {texture=/Game/Global/Textures/Default/Tileable_Patterns/Foam}.R) {constA=-1.7, constB=3.5})
  n27 = LinearInterpolate(
   A=Constant() {r=1}
   B=Constant()
   Alpha=MaterialFunctionCall(
    Transition Phase=@n17.B
    Height Texture=LinearInterpolate(A=@n21.RGB, B=@n24.RGB, Alpha=@n26)
    Contrast=ScalarParameter() {defaultValue=0.0025, group=WaterFall, parameterName=Waterfall Edge Contrast}
   ) {materialFunction=&HeightLerp}.Alpha
  )
  n28 = LinearInterpolate(
   B=Multiply(
    A=MaterialFunctionCall() {materialFunction=&DitherTemporalAA_Fortnite}.Result
    B=ScalarParameter() {defaultValue=28.2735, group=WaterFall, parameterName=WaterFall Depth jitter amount}
   ) {constB=10}
   Alpha=@n27
  ) {constB=100}
  n29 = Constant()
  n30 = ComponentMask(None=TextureCoordinate()) {r=true}
  n31 = Clamp(None=OneMinus(None=LinearInterpolate(Alpha=@n30) {constB=1.5}))
  n32 = Multiply(
   A=TextureCoordinate() {uTiling=0.4}
   B=AppendVector(A=ScalarParameter() {defaultValue=1, parameterName=Green_CustomXTile}, B=ScalarParameter() {defaultValue=1, parameterName=Green_CustomYTile})
  )
  n33 = MaterialFunctionCall(XY=@n32) {materialFunction=&Swizzle}
  n34 = Time() {bOverride_Period=true, period=9999999}
  n35 = Panner(
   Coordinate=Multiply(A=TextureCoordinate(), B=ScalarParameter() {defaultValue=0.7372, group=WaterFall_UVDistort, parameterName=UVDistortTilingA})
   Time=@n34
  ) {speedX=1.6, speedY=-0.08}
  n36 = TextureCoordinate() {uTiling=0.5, vTiling=11}
  n37 = ScalarParameter() {defaultValue=0.8, group=WaterFall_UVDistort, parameterName=UVDistortTilingB}
  n38 = TextureSample(UVs=Panner(Coordinate=Multiply(A=@n36, B=@n37), Time=@n34) {speedX=1, speedY=0.125}) {texture=/Game/Packages/Fortnite_UniqueMaterials_1/Textures/T_WaterFall_Tile}
  n39 = LinearInterpolate(
   A=Multiply(
    A=TextureSample(UVs=@n35) {texture=&T_WaterFall_Tile}.A
    B=ScalarParameter() {defaultValue=0.0163, group=WaterFall_UVDistort, parameterName=UVDistortAmountA}
   )
   B=Multiply(A=@n38.A, B=ScalarParameter() {defaultValue=0.0125, group=WaterFall_UVDistort, parameterName=UVDistortAmountB})
   Alpha=@n31
  )
  n40 = StaticSwitchParameter(True=Add(A=OneMinus(None=@n33.YX), B=@n39), False=@n32) {defaultValue=true, parameterName=UseWaterfallSwizzle}
  n41 = ScalarParameter() {defaultValue=1.8399, group=WaterFall, parameterName=FoamTile_GChannel}
  n42 = ScalarParameter() {defaultValue=0.5973, group=WaterFall, parameterName=Waterfall Speed}
  n43 = TextureSample(UVs=Panner(Coordinate=Multiply(A=@n40, B=@n41), Time=Multiply(A=@n34, B=@n42)) {speedY=-1}) {texture=&T_WaterFall_Tile}
  n44 = Multiply(A=VectorParameter() {group=WaterFall, parameterName=WaterFallColor, defaultValue=<rgba 0.2881 0.61 0.5573 1>}.RGB, B=@n43.G)
  n45 = Multiply(
   A=TextureCoordinate() {uTiling=0.7, vTiling=2.8}
   B=AppendVector(A=ScalarParameter() {defaultValue=1, parameterName=Red_CustomXTile}, B=ScalarParameter() {defaultValue=1, parameterName=Red_CustomYTile})
  )
  n46 = MaterialFunctionCall(XY=@n45) {materialFunction=&Swizzle}
  n47 = BumpOffset(Coordinate=Add(A=OneMinus(None=@n46.YX), B=@n39), Height=@n38.A) {heightRatio=0.125}
  n48 = Panner(
   Coordinate=Multiply(
    A=StaticSwitchParameter(True=@n47, False=@n45) {defaultValue=true, parameterName=UseWaterfallSwizzle}
    B=ScalarParameter() {defaultValue=1.0601, group=WaterFall, parameterName=FoamTile_RChannel}
   )
   Time=Multiply(A=@n34, B=ScalarParameter() {defaultValue=0.5973, group=WaterFall, parameterName=Waterfall Speed})
  ) {speedY=-1}
  n49 = TextureSample(UVs=@n48) {texture=&T_WaterFall_Tile}
  n50 = LinearInterpolate(A=@n44, B=Add(A=@n44, B=@n49.R), Alpha=ScalarParameter() {defaultValue=0.3, group=WaterFall, parameterName=LerpFoamA})
  n51 = Multiply(
   A=TextureCoordinate() {uTiling=0.333, vTiling=2}
   B=AppendVector(A=ScalarParameter() {defaultValue=1, parameterName=Blue_CustomXTile}, B=ScalarParameter() {defaultValue=1, parameterName=Blue_CustomYTile})
  )
  n52 = MaterialFunctionCall(XY=@n51) {materialFunction=&Swizzle}
  n53 = BumpOffset(Coordinate=Add(A=OneMinus(None=@n52.YX), B=@n39), Height=ScalarParameter() {defaultValue=0.8, group=WaterFall, parameterName=OffsetHeight}) {heightRatio=0.2}
  n54 = Panner(
   Coordinate=Multiply(
    A=StaticSwitchParameter(True=@n53, False=@n51) {defaultValue=true, parameterName=UseWaterfallSwizzle}
    B=ScalarParameter() {defaultValue=1, group=WaterFall, parameterName=FoamTile_BChannel}
   )
   Time=Multiply(A=@n34, B=ScalarParameter() {defaultValue=0.5973, group=WaterFall, parameterName=Waterfall Speed})
  ) {speedY=-1}
  n55 = TextureSample(UVs=@n54) {texture=&T_WaterFall_Tile}
  n56 = LinearInterpolate(A=@n50, B=Add(A=@n50, B=@n55.B), Alpha=ScalarParameter() {defaultValue=0.125, group=WaterFall, parameterName=LerpFoamB}) {constAlpha=0.4}
  n57 = Add(A=@n31, B=@n56)
  n58 = ScalarParameter() {defaultValue=0.3, group=WaterFall, parameterName=RoughnessMin}
  n59 = ScalarParameter() {defaultValue=1, group=WaterFall, parameterName=RoughnessMax}
  n60 = Multiply(A=Multiply(A=@n43.A, B=@n49.A), B=@n55.A)
  n61 = LinearInterpolate(
   A=TextureSample(UVs=@n54) {samplerType=SAMPLERTYPE_Normal, texture=/Game/Packages/Fortnite_UniqueMaterials_1/Textures/T_WaterFall_Tile_N}.RGB
   B=TextureSample(UVs=@n48) {samplerType=SAMPLERTYPE_Normal, texture=&T_WaterFall_Tile_N}.RGB
   Alpha=@n31
  )
  n62 = VectorParameter() {group=WaterFall, parameterName=ScaleNormal, defaultValue=<rgba 0.8 0.8 3.748 0>}
  n63 = MakeMaterialAttributes!(
   BaseColor=StaticSwitchParameter(True=Clamp(None=@n57), False=@n56) {parameterName=UseGradientOverlay}
   Metallic
   Specular=Constant() {r=0.5}
   Roughness=Clamp(None=LinearInterpolate(A=@n58, B=@n59, Alpha=@n60))
   Anisotropy
   EmissiveColor=Constant()
   Opacity
   OpacityMask
   Normal=Transform(None=Multiply(A=@n61, B=@n62.RGB))
   Tangent
   WorldPositionOffset
   SubsurfaceColor
   ClearCoat
   ClearCoatRoughness
   AmbientOcclusion
   Refraction
   CustomizedUVs_0
   CustomizedUVs_1
   CustomizedUVs_2
   CustomizedUVs_3
   CustomizedUVs_4
   CustomizedUVs_5
   CustomizedUVs_6
   CustomizedUVs_7
   PixelDepthOffset=StaticSwitchParameter(True=QualitySwitch(Default=@n28, Low=@n29, Medium=@n29), False=@n29) {defaultValue=true, parameterName=UseSTW_PDO}
   ShadingModel
   Displacement
  ) {desc=waterfall}
  n64 = ComponentMask(None=MaterialFunctionCall!(Material=@n63) {materialFunction=/Engine/Functions/MaterialLayerFunctions/MatLayerBlend_BreakBaseColor}.BaseColor) {g=true}
  n65 = Multiply(A=@n64, B=VectorParameter() {parameterName=GlowColor, defaultValue=<rgba 4.9919 0.1936 5 0>}.RGB)
  n66 = ComponentMask(None=WorldPosition().XYZ) {b=true}
  n67 = Divide(
   A=Subtract(A=@n66, B=ScalarParameter() {parameterName=WorldColorZOffset})
   B=ScalarParameter() {defaultValue=2000, parameterName=WorldColorZRange}
  )
  n68 = MaterialFunctionCall(
   Transition Phase=Multiply(A=Clamp(None=@n67), B=ScalarParameter() {defaultValue=1, parameterName=WorldColorZ_})
   Height Texture=@n64
   Contrast=ScalarParameter() {defaultValue=1, parameterName=WorldColorZ_Contrast}
  ) {materialFunction=&HeightLerp}
  n69 = ScalarParameter() {defaultValue=3, parameterName=WorldColor_EmissiveAmount}
  n70 = ScalarParameter() {parameterName=WorldColorZOffset_Gradient}
  n71 = Clamp(None=Divide(A=Subtract(A=@n66, B=@n70), B=ScalarParameter() {defaultValue=2000, parameterName=WorldColorZRange_1}))
  n72 = SetMaterialAttributes(
   MaterialAttributes=MaterialFunctionCall!(Material=@n63, NewBaseColor=@n65, Mask=OneMinus(None=@n68.Alpha)) {materialFunction=/Engine/Functions/MaterialLayerFunctions/MatLayerBlend_MultiplyBaseColor}.Blended Material
   Emissive Color=Multiply(A=Multiply(A=@n65, B=@n69), B=OneMinus(None=@n71))
  ) {attributeSetTypes [#1]}
  n73 = ScalarParameter() {defaultValue=0.1, group=Spec, parameterName=Spec Sand}
  n74 = ScalarParameter() {group=Roughness, parameterName=Sand Roughness}
  n75 = Constant()
  n76 = Multiply(A=@n10.RGB, B=VectorParameter() {parameterName=Bank_EdgeTint, defaultValue=<rgba 1 1 1 0>}.RGB)
  n77 = Multiply(A=ScalarParameter() {parameterName=Bank Edge Brightness}, B=@n10.RGB)
  n78 = MaterialFunctionCall(
   Transition Phase=Add(A=OneMinus(None=@n4), B=ScalarParameter() {group=Water Edge, parameterName=WaterEdgeOffset})
   Height Texture=@n12
   Contrast=ScalarParameter() {group=Water Edge, parameterName=Blend Contrast}
  ) {materialFunction=&HeightLerp}
  n79 = VertexColor()
  n80 = OneMinus(None=@n79.B)
  n81 = MaterialFunctionCall!(
   Height Offset Amount=@n80
   Texture mip bias amount=Multiply(A=@n79.B, B=ScalarParameter() {defaultValue=30, parameterName=Lake Blur amount})
   Ground texture mask=@n80
  ) {desc=Lake, materialFunction=/FigmentEnvironment/Packages/Fortnite_UniqueMaterials_1/DS_Fortnight/MF_Figment_ShallowWaterFunction}
  n82 = Add(A=@n78.Alpha, B=Multiply(A=@n81.Distance falloff, B=@n81.Distance falloff))
  n83 = VectorParameter() {parameterName=Sand Color, defaultValue=<rgba 0.5 0.5 1 1>}
  n84 = Multiply(
   A=Multiply(A=StaticSwitchParameter(True=@n76, False=@n77) {desc=HAACK, parameterName=Lake_Match}, B=Clamp(None=@n82))
   B=Multiply(A=OneMinus(None=@n4), B=Multiply(A=@n83.RGB, B=@n83.RGB))
  )
  n85 = TextureSampleParameter2D!(UVs=@n9, Apply View MipBias) {parameterName=RiverBedTexture1_Normal, samplerType=SAMPLERTYPE_Normal, texture=/Game/Packages/DS_Fortnight/pebbles_N}
  n86 = VertexColor()
  n87 = Clamp(None=Add(A=@n78.Alpha, B=ScalarParameter() {group=Water Edge, parameterName=WaterOpacity}))
  n88 = ScalarParameter() {defaultValue=0.4, group=Spec, parameterName=Spec Water Edge}
  n89 = Min(A=Add(A=Fresnel()) {constB=0.15})
  n90 = ScalarParameter() {group=Roughness, parameterName=Water Roughness}
  n91 = VertexNormalWS()
  n92 = Subtract(A=WorldPosition().XYZ, B=VectorParameter() {parameterName=CenterObjectLocation, defaultValue=<rgba 384 -128 -3072 1>}.RGB)
  n93 = Multiply(A=Multiply(A=Normalize(VectorInput=@n92), B=Constant3Vector() {constant=<rgba -1 1 1 0>}), B=Multiply(A=@n2.G) {constB=0.5})
  n94 = ConstantBiasScale(None=@n3.RGB) {bias=-0.5, scale=2}
  n95 = Transform(None=VectorParameter() {parameterName=testgf, defaultValue=<rgba -1 1 0 1>}.RGB) {transformSourceType=TRANSFORMSOURCE_Local}
  n96 = Multiply(
   A=MaterialFunctionCall() {materialFunction=&TangentBasis}.X U
   B=Constant3Vector() {constant=<rgba -1 1 1 0>}
  )
  n97 = ScalarParameter() {group=FlowMap, parameterName=Pebble Flow Center Mask}
  n98 = LinearInterpolate(
   A=Multiply(A=VectorParameter() {group=FlowMap, parameterName=Flow Pebble Normal}.RGB, B=@n85.RGB)
   Alpha=Min(A=Multiply(A=@n3.A, B=@n97) {constB=1.3})
  ) {constB=0}
  n99 = Constant()
  n100 = Add(
   A=StaticSwitchParameter(
    True=Transform(None=@n94) {desc="!!!", transformSourceType=TRANSFORMSOURCE_World, transformType=TRANSFORM_Local}
    False=StaticSwitchParameter(True=@n95, False=@n96) {parameterName=UseCenterObjectDirection}
   ) {defaultValue=true, parameterName=Use Flow Texturemap}
   B=QualitySwitch(Default=@n98, Low=@n99, High=@n98, Medium=@n99)
  )
  n101 = LinearInterpolate(A=@n100, B=Multiply(A=@n100, B=Multiply(A=@n2.G, B=@n4)) {desc=Mask Edges}, Alpha=ScalarParameter() {group=FlowMap, parameterName=Bank Drag})
  n102 = VectorParameter() {group=FlowMap, parameterName=Flow, defaultValue=<rgba 1 1 0 0>}
  n103 = StaticSwitchParameter(True=@n93, False=Multiply(A=QualitySwitch(Default=@n101, Low=@n100, High=@n101, Medium=@n100), B=@n102.RGB)) {parameterName=UseCenterObjectDirection}
  n104 = WorldPosition()
  n105 = MaterialFunctionCall!(
   Diffuse=TextureObject() {texture=/Game/Packages/Fortnite_UniqueMaterials_1/Textures/FoamHeight}
   Normal Map=TextureObject() {samplerType=SAMPLERTYPE_Normal, texture=/Game/Packages/Fortnite_UniqueMaterials_1/Textures/FoamNormals}
   "Flow Vector Map (see tooltip)"=ComponentMask(None=@n103) {g=true, r=true}
   UV texture 2 offset=ComponentMask(None=VectorParameter() {parameterName=Offset}.RGB) {g=true, r=true}
   "Time (see tooltip)"=Multiply(A=Time() {bOverride_Period=true, period=9999999}, B=ScalarParameter() {group=FlowMap, parameterName=Speed})
   UVs=Divide(A=ComponentMask(None=@n104.XYZ) {g=true, r=true}, B=ScalarParameter() {defaultValue=512, parameterName=Size})
  ) {materialFunction=/Engine/Functions/Engine_MaterialFunctions02/Texturing/FlowMaps_Simple}
  n106 = Multiply(A=@n105.Normal, B=VectorParameter() {parameterName=Normal Intensity, defaultValue=<rgba 0.3 0.3 1 0>}.RGB)
  n107 = VectorParameter() {group=Refraction, parameterName=Refraction, defaultValue=<rgba 0.3 0.3 1 0>}
  n108 = Multiply(A=ComponentMask(None=Multiply(A=@n105.Normal, B=@n107.RGB)) {g=true, r=true}, B=@n4) {desc=Bank Mask}
  n109 = Add(A=Multiply(A=@n108, B=ScalarParameter() {group=Refraction, parameterName=Refraction Min}) {constB=0.5}, B=@n9)
  n110 = StaticSwitchParameter(True=@n93, False=Multiply(A=@n100, B=@n102.RGB)) {parameterName=UseCenterObjectDirection}
  n111 = ComponentMask(None=@n110) {g=true, r=true}
  n112 = Constant2Vector() {g=1, r=-1}
  n113 = Multiply(
   A=StaticSwitchParameter(True=ComponentMask(None=@n110) {g=true, r=true}, False=Multiply(A=@n111, B=@n112)) {parameterName=UseCenterObjectDirection}
   B=ScalarParameter() {group=FlowMap, parameterName=Flow WPO Shift}
  )
  n114 = Constant()
  n115 = TextureSampleParameter2D!(
   UVs=Add(A=QualitySwitch(Default=@n113, Low=@n114, High=@n113, Medium=@n114), B=Add(A=@n108, B=@n9) {desc=Distortion}) {desc=Adds flow offset}
   Bias=ScalarParameter() {group=Refraction, parameterName=Refract Mip Bias}
   Apply View MipBias
  ) {mipValueMode=TMVM_MipBias, parameterName=RiverBedTexture_2, texture=/Game/Packages/Fortnite_UniqueMaterials_1/Textures/T_River_PebbleFoam_01}
  n116 = Multiply(
   A=MaterialFunctionCall!(Alpha=@n4, Offset=ScalarParameter() {group=Refraction, parameterName=RefractionAlphaOffset}) {materialFunction=/Engine/Functions/Engine_MaterialFunctions01/Opacity/AlphaOffset}.Result
   B=ScalarParameter() {defaultValue=4, group=Refraction, parameterName=Refraction Mask Mult}
  )
  n117 = Multiply(A=LinearInterpolate(
    A=TextureSampleParameter2D(UVs=@n109) {mipValueMode=TMVM_MipBias, parameterName=RiverBedTexture_1, texture=&pebbles}.RGB
    B=@n115.RGB
    Alpha=Clamp(None=@n116)
   ))
  n118 = VectorParameter() {parameterName=Water High, defaultValue=<rgba 0.5 0.5 1 1>}
  n119 = VectorParameter() {parameterName=Water Low, defaultValue=<rgba 0.1793 0.3574 1 0.2>}
  n120 = Multiply(
   A=QualitySwitch(Default=@n117, Low=@n115.RGB, High=@n117, Medium=@n115.RGB)
   B=LinearInterpolate(A=Multiply(A=@n118.RGB, B=@n118.RGB), B=Multiply(A=@n119.RGB, B=@n119.RGB), Alpha=@n4)
  )
  n121 = LinearInterpolate(
   A=ScalarParameter() {defaultValue=1, group=Fresnel, parameterName=Opacity Min}
   B=ScalarParameter() {defaultValue=1, group=Fresnel, parameterName=Opacity Max}
   Alpha=Fresnel()
  )
  n122 = MaterialFunctionCall!(In=MakeMaterialAttributes(
    BaseColor=Multiply(A=@n120, B=@n121)
    Specular=Multiply(A=@n88, B=@n89)
    Roughness=Clamp(None=@n90)
    EmissiveColor=Constant()
    Normal=MaterialFunctionCall(BaseNormal=@n91, AdditionalNormal=@n106) {materialFunction=&BlendAngleCorrectedNormals}.Result
   ) {desc=Stream water}) {materialFunction=/Game/Athena/Environments/Landscape/MaterialFunctions/MF_River_Tint_Add}
  n123 = StaticSwitchParameter(
   True=MaterialFunctionCall!(Base Material=@n81.MA Result, Top Material=@n122.Result, Alpha=@n79.B) {materialFunction=/Engine/Functions/MaterialLayerFunctions/MatLayerBlend_Standard}.Blended Material
   False=@n122.Result
  ) {desc=+100 instructions, group=Lake, parameterName=Enable Transition To Lake?}
  n124 = StaticSwitchParameter(
   True=MaterialFunctionCall(
    Base Material=@n123
    Top Material=MakeMaterialAttributes(BaseColor=@n84, Specular=@n73, Roughness=@n74, EmissiveColor=@n75, Normal=@n85.RGB) {desc=Sand}
    Alpha=Multiply(A=@n86.B, B=@n87) {constA=1}
   ) {desc=Sand, materialFunction=&MatLayerBlend_Standard}.Blended Material
   False=@n123
  ) {defaultValue=true, parameterName=UseSand}
  n125 = StaticSwitchParameter(
   True=MaterialFunctionCall(Base Material=@n124, Top Material=StaticSwitchParameter(True=@n72, False=@n63) {parameterName=UseWorldColorGlow}, Alpha=@n27) {materialFunction=&MatLayerBlend_Standard}.Blended Material
   False=@n124
  ) {parameterName=UseWaterFall}
  n126 = CameraPositionWS()
  n127 = WorldPosition()
  n128 = Subtract(A=@n126, B=@n127.XYZ)
  n129 = ScalarParameter() {defaultValue=200, parameterName=DistanceZFightBias}
  n130 = SphereMask(
   A=@n126
   B=@n127.XYZ
   Radius=ScalarParameter() {defaultValue=50000, parameterName=ZBiasStart}
   Hardness=ScalarParameter() {defaultValue=0.5, parameterName=ZBiasHardness}
  )
  n131 = Add(A=Multiply(A=Normalize(VectorInput=@n128), B=LinearInterpolate(A=@n129, Alpha=@n130)), B=Constant())
  n132 = VertexColor()
  MaterialExpressionAppend3Vector_0 = MaterialXAppend3Vector!(A=Constant(), B=Constant(), C=ScalarParameter() {defaultValue=-3, group=Rufus, parameterName=Water_Z_Offset})
  n133 = Multiply(A=OneMinus(None=Multiply(A=@n132.R, B=@n132.R)), B=@MaterialExpressionAppend3Vector_0)
  n134 = MaterialFunctionCall() {materialFunction=&LinearGradient}
  n135 = LinearInterpolate(
   A=ScalarParameter() {group=Rufus, parameterName=Rufus_WPO_EdgeOffset}
   B=@n133
   Alpha=StaticSwitchParameter(True=OneMinus(None=@n134.UGradient), False=@n134.UGradient) {group=Rufus, parameterName=Flip?}
  ) {constB=0}
  n136 = StaticSwitchParameter(
   True=@n131
   False=StaticSwitchParameter(True=StaticSwitchParameter(True=@n135, False=@n133) {group=Rufus, parameterName=EnableWPO_Adjust}, False=@n131) {desc=+100 instructions, group=Lake, parameterName=Enable Transition To Lake?}
  ) {parameterName=UseWaterFall}
  n137 = ComponentMask(None=Multiply(A=@n103, B=Constant3Vector() {constant=<rgba 1 -1 1 0>})) {g=true, r=true}
  n138 = Multiply(A=Add(A=Clamp(None=@n137) {maxDefault=0.3, minDefault=-0.3}) {constB=0.3}) {constB=1.5}
  n139 = LinearInterpolate(A=Constant2Vector() {g=0.5, r=0.5}, B=@n138, Alpha=VertexColor().B) {constB=0.5}
  n140 = SetMaterialAttributes!(
   MaterialAttributes
   Base Color=MaterialFunctionCall() {materialFunction=/Game/Athena/WinterArt/Materials/MF_SharedWinter}.IceColor
   Roughness=ScalarParameter() {defaultValue=0.124}
   Specular=ScalarParameter() {defaultValue=0.2, parameterName=Param_1}
   Normal=Constant3Vector() {constant=<rgba 0 0 1 0>}
  ) {attributeSetTypes [
    #2
    #3
    #4
    #5
   ]}
  n141 = TextureCoordinate() {uTiling=0.5, vTiling=0.5}
  n142 = Subtract(A=ComponentMask(None=Sine(None=@n141)) {g=true}, B=ScalarParameter() {defaultValue=0.3, parameterName=Param_2})
  BlendMaterialAttributes(
   A=MaterialFunctionCall!(
    Render To Texture Pass
    WS Flow Map Dir=StaticSwitchParameter(True=@n139, False=@n138) {desc=+100 instructions, group=Lake, parameterName=Enable Transition To Lake?}
    Opacity=StaticSwitchParameter(True=@n1, False=@n16) {parameterName=UseWaterFall}
    Default Mat Att=SetMaterialAttributes!(MaterialAttributes=@n125, Opacity Mask=@n15, World Position Offset=@n136) {attributeSetTypes [#6, #7]}
   ) {materialFunction=/Game/WaterAndWind/Materials/Water/WaterRenderToTextureMaterialAtt}.Result
   B=@n140
   Alpha=Saturate(None=Round(None=@n142))
  )
  Clamp(None=Add(A=@n17.B))
  n143 = Rotator(
   Coordinate=Multiply(A=TextureCoordinate() {coordinateIndex=1}, B=ScalarParameter() {defaultValue=1, parameterName=2ndUVset_Scale})
   Time=Time()
  ) {speed=-1}
  n144 = LinearInterpolate(
   A=ScalarParameter() {parameterName=WhirlpoolMin}
   B=ScalarParameter() {defaultValue=0.6787, parameterName=WhirlpoolMax}
   Alpha=Multiply(
    A=@n27
    B=TextureSampleParameter2D!(UVs=@n143, Apply View MipBias) {parameterName=2ndUVMask, samplerType=SAMPLERTYPE_LinearColor, texture=/Game/Athena/Environments/Textures/T_Athena_Whirlpool_M}.R
   )
  )
  n145 = StaticSwitchParameter(True=Clamp(None=Subtract(A=@n27, B=@n144)), False=@n27) {parameterName=Use2ndUVMask}
  FeatureLevelSwitch(Default=@n27, ES2=@n27, ES3_1=@n27, SM4_REMOVED=@n27, SM5=@n145, SM6=@n145)
  LinearInterpolate(B=VertexNormalWS())
  n146 = Multiply(
   A=MaterialFunctionCall() {materialFunction=&TangentBasis}.X U
   B=Constant3Vector() {constant=<rgba -1 1 1 0>}
  )
  n147 = Transform(None=Constant3Vector() {constant=<rgba 0 -1 0 0>}) {transformSourceType=TRANSFORMSOURCE_Local}
  Reroute(None=StaticSwitchParameter(True=@n146, False=StaticSwitchParameter(True=@n147, False=@n146) {parameterName=UseWaterFall}) {desc=HAACK, parameterName=Lake_Match}) {desc=Previously Connected to useCenterObjectDirection: False}
  VectorParameter()
  VertexColor()
 ]
}
