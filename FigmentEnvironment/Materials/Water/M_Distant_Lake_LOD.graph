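# Distant LOD lake material for Loot Lake, layering a glowing purple corruption effect over a shared shallow-water function.
# MPC_Landscape's LakeCorruption drives a world-space sphere mask's radius (R), emissive strength (G) and pulse color (B).
# Lake-mask and electric-noise textures shape the glow, with Fresnel replaced by a constant on mobile.
# Material-layer blends add emissive, blacken base color and reduce specular in corrupted areas, override roughness and flatten normals.
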Material /FigmentEnvironment/Materials/Water/M_Distant_Lake_LOD
graph {
 properties {
  bTangentSpaceNormal=false
  bUseMaterialAttributes=true
  bUsedWithInstancedStaticMeshes=true
  bUsedWithNanite=true
  pixelDepthOffsetMode=PDOM_Legacy
  usedShadingModels=Default Lit
 }
 nodes [
  n1 = Multiply(
   A=VectorParameter() {parameterName=PulseColor, defaultValue=<rgba 2 0.1738 1.8546 0>}.RGB
   B=ComponentMask(None=CollectionParameter() {collection=/Game/Athena/Environments/Landscape/MPC/MPC_Landscape, parameterName=LakeCorruption}) {b=true}
  )
  n2 = Add(
   A=Multiply(A=@n1) {constB=0.2}
   B=ShadingPathSwitch(Default=Fresnel() {baseReflectFraction=0.15, exponent=9}, Mobile=Constant() {r=0.35}) {desc=Migrated from FeatureLevelSwitch}
  ) {constB=0.25}
  n3 = TextureSample(UVs=Multiply(A=TextureCoordinate(), B=ScalarParameter() {defaultValue=4.2, parameterName=MacroTile})) {samplerType=SAMPLERTYPE_LinearColor, texture=/Game/Characters/Player/Male/Medium/Bodies/M_MED_SRDrift/Textures/T_ElectricNoise}
  n4 = Add(A=@n3.G) {constB=0.3}
  n5 = TextureSample() {texture=/Game/Athena/Prototype/Blueprints/Cube/Materials/T_Loot_Lake_Mask}
  n6 = Multiply(
   A=Add(A=@n1)
   B=Multiply(A=LinearInterpolate(A=@n4, Alpha=@n5.R), B=VectorParameter() {parameterName=DistantGlowColor, defaultValue=<rgba 0.4855 0.0781 1 0>}.RGB)
  )
  n7 = Multiply(A=@n5.R, B=VectorParameter() {parameterName=DimGlow, defaultValue=<rgba 0.0259 0.0389 0.15 1>}.RGB)
  n8 = LinearInterpolate(A=ScalarParameter() {defaultValue=0.75, parameterName=MacroMaskMin}, Alpha=@n3.G) {constA=0.75}
  n9 = CollectionParameter() {collection=&MPC_Landscape, parameterName=LakeCorruption}
  n10 = SphereMask(
   A=ComponentMask(None=WorldPosition().XYZ) {g=true, r=true}
   B=ComponentMask(None=VectorParameter() {parameterName=CorruptionCenter, defaultValue=<rgba 63281 -49074 0 0>}.RGB) {g=true, r=true}
   Radius=Multiply(A=ScalarParameter() {defaultValue=28000, parameterName=MaxCorruptionRadius}, B=ComponentMask(None=@n9) {r=true})
  ) {hardnessPercent=70}
  n11 = TextureSample() {texture=&T_Loot_Lake_Mask}
  n12 = Multiply(A=Multiply(A=Multiply(A=@n2) {constB=2.5}, B=Add(A=@n6, B=@n7) {constB=0}), B=Multiply(A=Multiply(A=@n8, B=@n10), B=@n11.G))
  n13 = Multiply(
   A=ScalarParameter() {defaultValue=1.2, parameterName=EmissiveMult}
   B=ComponentMask(None=CollectionParameter() {collection=&MPC_Landscape, parameterName=LakeCorruption}) {g=true}
  )
  n14 = Multiply(A=@n11.G, B=@n10)
  n15 = MaterialFunctionCall!(
   Material=MaterialFunctionCall!(
    Material=MaterialFunctionCall!(Height Offset Amount, Texture mip bias amount, Ground texture mask) {materialFunction=&/MF_Figment_ShallowWaterFunction}.MA Result
    Emissive=Multiply(A=@n12, B=@n13)
   ) {materialFunction=/Engine/Functions/MaterialLayerFunctions/MatLayerBlend_Emissive}.Blended Material
   NewBaseColor=Constant()
   Mask=@n14
  ) {materialFunction=/Engine/Functions/MaterialLayerFunctions/MatLayerBlend_MultiplyBaseColor}
  n16 = LinearInterpolate(B=ScalarParameter() {defaultValue=0.25, parameterName=CorruptSpecularMult}, Alpha=@n14) {constA=1}
  MaterialFunctionCall!(
   Material=MaterialFunctionCall!(
    Material=MaterialFunctionCall!(Base Material=@n15.Blended Material, SpecularMultiplier=@n16) {materialFunction=/Engine/Functions/MaterialLayerFunctions/MatLayerBlend_ModulateSpecular}.Blended Material
    NewRoughness=ScalarParameter() {defaultValue=0.05, parameterName=R}
    Mask
   ) {materialFunction=/Engine/Functions/MaterialLayerFunctions/MatLayerBlend_RoughnessOverride}.Blended Material
   Normal=Constant3Vector() {constant=<rgba 0 0 1 0>}
  ) {materialFunction=/Engine/Functions/MaterialLayerFunctions/MatLayerBlend_ReplaceNormals}
  MaterialFunctionCall(Speed=Constant() {r=0.05}) {materialFunction=&DebugTimeSine}
  StaticBool() {value=true}
 ]
}
